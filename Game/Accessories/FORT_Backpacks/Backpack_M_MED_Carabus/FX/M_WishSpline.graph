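# Translucent material for animated, wispy spline strands on the Carabus backpack, tinted by CC_SecondaryColor.
# Vertex offset combines SplineThicken, camera and local offsets, and a UV-weighted sine sway phased by height, time and mesh side.
# Panning noise and sphere masks form a wisp mask that drives an HSV shift of the tint, blended by local height.
# Opacity multiplies that mask by UV and height fades, depth fade, and the shifted color's blue channel, remapped.
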Material /Game/Accessories/FORT_Backpacks/Backpack_M_MED_Carabus/FX/M_WishSpline
graph {
 properties {
  bUsedWithMeshDeformer=true
  bUsedWithNiagaraMeshParticles=true
  bUsedWithSkeletalMesh=true
  blendMode=BLEND_Translucent
  pixelDepthOffsetMode=PDOM_Legacy
  usedShadingModels=Default Lit
 }
 nodes [
  n1 = ScalarParameter() {defaultValue=10, parameterName=CamOffset}
  n2 = MaterialFunctionCall(
   WidthBase=ScalarParameter() {defaultValue=0.5, parameterName=Width_Base}
   WidthTip=ScalarParameter() {defaultValue=40, parameterName=Width_End}
  ) {materialFunction=&SplineThicken}
  n3 = OneMinus(None=ComponentMask(None=TextureCoordinate()) {g=true})
  n4 = TransformPosition(None=WorldPosition().XYZ) {transformSourceType=TRANSFORMPOSSOURCE_World}
  n5 = Sign(None=ComponentMask(None=PreSkinnedPosition()) {r=true})
  n6 = Constant() {r=0.125}
  n7 = Time()
  n8 = Constant() {r=0.5}
  n9 = Add(
   A=Multiply(A=ComponentMask(None=@n4) {b=true}, B=Constant() {r=0.02})
   B=Add(A=Multiply(A=@n5, B=@n6), B=Multiply(A=@n7, B=@n8) {constB=0.5})
  )
  n10 = Multiply(A=@n3, B=Multiply(A=Sine(None=@n9), B=VectorParameter() {parameterName=Param_1, defaultValue=<rgba 15 0 0 0>}.RGB))
  Add(
   A=Add(
    A=MaterialFunctionCall!(Camera Offset=@n1, Additional World Position Offset) {materialFunction=&/MF_CameraOffset}.World Position Offset
    B=Add(A=@n2.WorldPositionOffset, B=@n10)
   )
   B=Transform(None=VectorParameter() {parameterName=Offset, defaultValue=<rgba 0 0.75 1.5 0>}.RGB) {transformSourceType=TRANSFORMSOURCE_Local}
  )
  n11 = VertexInterpolator(VS=@n5)
  n12 = Multiply(
   A=Add(A=TextureCoordinate(), B=Multiply(A=@n11.PS, B=@n6))
   B=ComponentMask(None=Constant3Vector() {constant=<rgba 0.5 1 0 0>}) {g=true, r=true}
  )
  n13 = Time()
  n14 = Panner(Coordinate=@n12, Time=@n13) {speedY=0.7}
  n15 = MaterialFunctionCall(
   Input=TextureSample(UVs=Multiply(A=@n14) {constB=4}) {texture=&T_NoisePack_01}.B
   Input Low=Constant()
   Input High=Constant() {r=1}
   Target Low=Constant() {r=0.7}
   Target High=Constant() {r=1}
  ) {materialFunction=&RemapValueRange}
  n16 = Panner(Coordinate=@n12, Time=@n13) {speedY=0.4}
  n17 = Panner(Coordinate=@n12, Time=@n13) {speedY=0.8}
  n18 = OneMinus(None=Max(
    A=TextureSample(UVs=@n16) {texture=&T_NoisePack_01}.R
    B=TextureSample(UVs=@n17) {texture=&T_NoisePack_01}.G
   ))
  n19 = TextureCoordinate() {vTiling=0}
  n20 = Constant2Vector() {r=0.5}
  n21 = Subtract(A=Multiply(A=Multiply(A=@n15.Result, B=@n18), B=SphereMask(A=@n19, B=@n20) {attenuationRadius=0.5, hardnessPercent=25}), B=@n3)
  n22 = Constant() {r=0.5403}
  n23 = Multiply(A=Saturate(None=Divide(A=@n21, B=@n22)), B=Constant() {r=1.5})
  n24 = SphereMask(A=TextureCoordinate() {uTiling=0}, B=Constant2Vector()) {attenuationRadius=1, hardnessPercent=95}
  n25 = Multiply(A=Saturate(None=Multiply(A=@n23, B=@n24))) {constB=0.95}
  n26 = AppendVector(A=Multiply(A=OneMinus(None=@n25)) {constB=-0.1}, B=Constant() {r=-0.25})
  n27 = Constant() {r=0.15}
  n28 = VectorParameter() {parameterName=CC_SecondaryColor, defaultValue=<rgba 1 1 1 0>}
  n29 = MaterialFunctionCall!(HSV_In=Add(
    A=AppendVector(A=@n26, B=@n27)
    B=MaterialFunctionCall!(RGB=@n28.RGB) {materialFunction=/Engine/Functions/Engine_MaterialFunctions02/Math/RGBtoHSV}."RGB to HSV (0-1)"
   )) {materialFunction=/Game/Athena/UI/Materials/MF_HSVtoRGB}
  n30 = WorldPosition()
  n31 = Subtract(A=ComponentMask(None=TransformPosition(None=@n30.XYZ) {transformSourceType=TRANSFORMPOSSOURCE_World}) {b=true}, B=Constant() {r=-70})
  n32 = Constant() {r=40}
  LinearInterpolate(A=@n29.RGB_Out, B=@n28.RGB, Alpha=Saturate(None=Divide(A=@n31, B=@n32)))
  n33 = OneMinus(None=ComponentMask(None=TextureCoordinate()) {g=true})
  n34 = Constant()
  n35 = Saturate(None=Divide(A=Subtract(A=@n33, B=@n34), B=Constant() {r=0.1}))
  n36 = WorldPosition()
  n37 = Subtract(A=ComponentMask(None=TransformPosition(None=@n36.XYZ) {transformSourceType=TRANSFORMPOSSOURCE_World}) {b=true}, B=Constant() {r=-70})
  n38 = Constant() {r=35}
  n39 = Multiply(A=Saturate(None=Divide(A=@n37, B=@n38)), B=@n25)
  Multiply(
   A=DepthFade(Opacity=Multiply(A=@n35, B=@n39), FadeDistance=Constant() {r=8}) {fadeDistanceDefault=8}
   B=MaterialFunctionCall(
    Input=MaterialFunctionCall(Float3=@n29.RGB_Out) {materialFunction=&BreakOutFloat3Components}.B
    Input Low=Constant() {r=0.3}
    Input High=Constant() {r=1}
    Target Low=Constant() {r=1}
    Target High=Constant() {r=0.6}
   ) {materialFunction=&RemapValueRange}.Result
  )
 ]
}
